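; BP_TestRemoveNavMeshBounds is an AI test Blueprint for removing nav mesh bounds at runtime through World Partition data layers.
; Its custom event DisableNavBoundsDL gets the Data Layer Manager, resolves the TestRemoveNavMesh data layer asset to its instance, and sets that instance to Unloaded non-recursively.
; BeginPlay, ActorBeginOverlap, Tick and the construction script are empty, and nothing in the graph calls the custom event.
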(event EventBeginPlay)

(event Collision|EventActorBeginOverlap
  (params (OtherActor :type Actor%20Object%20Reference)))

(event EventTick
  (params (DeltaSeconds :type Float%20%28single-precision%29)))

(event Custom|DisableNavBoundsDL
  (bind _returnvalue (WorldPartition|GetDataLayerManager))
  (DataLayers|SetDataLayerInstanceRuntimeState _returnvalue (DataLayers|GetDataLayerInstancefromAsset _returnvalue "/Game/Athena/Maps/Test/AI/DataLayers/AI_WP_DataLayer_TestRemoveNavMesh.AI_WP_DataLayer_TestRemoveNavMesh") "Unloaded" false))

(fn ConstructionScript)
